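# Niagara trail system for the InflatoDodo glider, combining spinning propellers, speed-line sprites and two wingtip ribbon trails.
# Prop pins two delayed-scale-in meshes to sockets, crossfading solid props into tinted blur and glow as a smoothed system value rises.
# SpeedLines and Wingtrails scale spawning, opacity or velocity from system timelines, with time-of-day tinting or scaling.
# Front-end modifiers substitute separate values for menu previews throughout.
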NiagaraSystem /BRCosmetics/Gadgets/Assets/VinderTech_GliderChute/Glider_InflatoDodo/FX/NS_Glider_InflatoDodo_Trail
dependencies {
 usedDataInterfaces [
  &NiagaraDataInterfaceCurve
  &NiagaraDataInterfaceArrayInt32
  &NiagaraDataInterfaceVector2DCurve
  &NiagaraDataInterfaceColorCurve
  &NiagaraDataInterfaceSocketReader
  &NiagaraDataInterfaceArrayFloat
 ]
 usedDynamicInputs [
  &RemapRange
  &SelectFloatFromArray
  &MakeIntFromFloat
  &NormalizeVector
 ]
 usedModules [
  &self:SystemSpawnScript.SystemScriptSource.SystemScriptGraph.NiagaraNodeAssignment_1.SetVariables_#1
  &SystemState
  &Timeline
  &FN_Integrator
 ]
}
emitters {
 Prop {
  emitterUpdateScript [
   &EmitterState(Life Cycle Mode=::NewEnumerator1, Loop Behavior=::NewEnumerator1, Loop Duration Mode=::NewEnumerator1, UseLoopDelay=false)
   &SpawnBurst_Instantaneous(Spawn Count=2, Use Spawn Probability=false, UseLoopCountLimit=false)
   /SetVariables(Emitter.EmisColor:LinearColor=<rgba 5 2.7883 0 1>) {isSetParameters=true}
   /SetVariables(Emitter.ScaleDelay:NiagaraFloat=&FloatFromCurve(
     FloatCurve=DataInterface() {propertyValues {Curve {keys [{tangentMode=RCTM_User, time=1.2, value=0}, {tangentMode=RCTM_User, time=1.35, value=1}]}}}
     CurveIndex=$Emitter.Age
    )) {isSetParameters=true}
  ]
  particleSpawnScript [
   &InitializeParticle(
    Color Mode=::NewEnumerator0
    Lifetime=5
    Mesh Scale Mode=::NewEnumerator3
    Mesh Uniform Scale=0.9
    UsePositionOffset=false
    Write Material Random=true
   )
   /SetVariables(Particles.MotionBlurScale:Vector3f=<0.5 0.5 0.1>, Particles.MotionGlowScale:Vector3f=<0.34 0.34 0.1>) {isSetParameters=true}
   /SetVariables(Particles.MotionBlurDynamic:Vector4f=&MakeVector4(X=-4, Y=6, Z=0, W=0)) {isSetParameters=true}
   &PartitionParticles()
   &InitialMeshOrientation(Mesh Orientation Mode=::NewEnumerator2, Use Rotation Vector=false)
  ]
  particleUpdateScript [
   &ParticleState(Kill Particles When Lifetime Has Elapsed=false, Let Infinitely Lived Particles Die When Emitter Deactivates=true)
   &ScaleMeshSize(Scale Factor=&Multiply_VectorByFloat(Vector=<1 1 1>, Float=$Emitter.ScaleDelay))
   &SocketLocation(
    Orientation Sampling=::NewEnumerator1
    SamplePosOffset=true
    Socket Index=&SelectIntFromArray(Direct Array Index=$Particles.PartitionParticles.Partition, Int Selection Array=DataInterface() {propertyValues {IntData [4,5]}})
    Socket Reader=$System.Socket Reader
    Socket Sampling Mode=::NewEnumerator2
    VelocitySpeedLimit=false
   )
   &UpdateMeshOrientation(
    Rotation Coordinate Space=::NewEnumerator0
    Rotation Rate=&FN_FrontEndFloatModifier(
     Gameplay=&Lerp_Float(A=3, B=10, Alpha=$System.InterpolateOverTime002.InterpolateOverTime_Float.Moving Average)
     Front End=10
    )
    Rotation Vector=<0 0 1>
   )
   /SetVariables(
    Particles.BlurPropColor:LinearColor=&MakeLinearColorFromVectorAndFloat(
     "Vector (RGB)"=&MakeVectorFromLinearColorRGB(Color=&FNTimeOfDayColor(Night Color=<rgba 0.8 0.229 0 1>, Morning Color=<rgba 0.8 0.229 0 1>, Day Color=<rgba 1 0.2377 0 1>, Evening Color=<rgba 0.8 0.229 0 1>))
     "Float (Alpha)"=&FN_FrontEndFloatModifier(
      Gameplay=&FloatFromCurve(
       FloatCurve=DataInterface() {propertyValues {Curve {keys [{tangentMode=RCTM_User, time=0, value=0}, {tangentMode=RCTM_User, time=0.25, value=1}]}}}
       CurveIndex=$System.InterpolateOverTime002.InterpolateOverTime_Float.Moving Average
      )
      Front End=1
     )
    )
    Particles.BlurPropGlowColor:LinearColor=&MakeLinearColorFromVectorAndFloat(
     "Vector (RGB)"=<1 1 1>
     "Float (Alpha)"=&Multiply_Float(
      A=0.05
      B=&FN_FrontEndFloatModifier(
       Gameplay=&FloatFromCurve(
        FloatCurve=DataInterface() {propertyValues {Curve {keys [{tangentMode=RCTM_User, time=0, value=0}, {tangentMode=RCTM_User, time=0.25, value=1}]}}}
        CurveIndex=$System.InterpolateOverTime002.InterpolateOverTime_Float.Moving Average
       )
       Front End=1
      )
     )
    )
    Particles.MainPropColor:LinearColor=&MakeLinearColorFromVectorAndFloat(
     "Vector (RGB)"=<1 1 1>
     "Float (Alpha)"=&FN_FrontEndFloatModifier(
      Gameplay=&FloatFromCurve(
       FloatCurve=DataInterface() {propertyValues {Curve {keys [{interpMode=RCIM_Cubic, tangentMode=RCTM_User, time=0, value=1}, {interpMode=RCIM_Cubic, tangentMode=RCTM_User, time=0.25, value=0}]}}}
       CurveIndex=$System.InterpolateOverTime002.InterpolateOverTime_Float.Moving Average
      )
      Front End=0
     )
    )
   ) {isSetParameters=true}
   /SetVariables(
    Particles.MotionBlurScale:Vector3f=&Multiply_VectorByFloat(Vector=$Particles.Initial.MotionBlurScale, Float=$Emitter.ScaleDelay)
    Particles.MotionGlowScale:Vector3f=&Multiply_VectorByFloat(Vector=$Particles.Initial.MotionGlowScale, Float=$Emitter.ScaleDelay)
   ) {isSetParameters=true}
   /SetVariables(Particles.BlurPos:NiagaraPosition=&AddVectorToPosition(
     Position=$Output.SocketLocation.SampledPosition
     Vector=&TransformVector(Vector=<0 0 -5>, Destination Space=::Simulation, Transform Vector as Position=true)
    )) {isSetParameters=true}
  ]
  rendererClasses [&NiagaraMeshRendererProperties]
  renderers [
   {rendererClass=&NiagaraMeshRendererProperties, rendererIndex=0}
   {rendererClass=&NiagaraMeshRendererProperties, rendererIndex=1}
   {rendererClass=&NiagaraMeshRendererProperties, rendererIndex=2}
  ]
 }
 SpeedLines {
  emitterUpdateScript [
   &EmitterState(
    Delay First Loop Only=true
    Inactive Response=::NewEnumerator1
    Life Cycle Mode=::NewEnumerator1
    Loop Delay=1
    Loop Duration=1
    UseLoopDelay=true
   )
   &SpawnRate(
    SpawnRate=&Multiply_Float(
     A=&FN_FrontEndFloatModifier(Gameplay=&Multiply_Float(A=100, B=$System.Timeline.Time), Front End=50)
     B=1
    )
    Use Spawn Probability=false
   ) {module=SpawnRate001}
  ]
  particleSpawnScript [
   &InitializeParticle(
    Color=&MakeLinearColorFromVectorAndFloat(
     "Vector (RGB)"=&MakeVectorFromLinearColorRGB(Color=<rgba 1 1 1 1>)
     "Float (Alpha)"=&FN_FrontEndFloatModifier(Gameplay=0.25, Front End=0.1)
    )
    Lifetime Max=0.16
    Lifetime Min=0.1
    Lifetime Mode=::NewEnumerator1
    Sprite Size Max=<4 10>
    Sprite Size Min=<3 5>
    Sprite Size Mode=::NewEnumerator2
    UsePositionOffset=true
    Write Material Random=true
   )
   &ShapeLocation(
    Handle Radius=5
    Large Radius=&FN_FrontEndFloatModifier(Gameplay=330, Front End=250)
    MaskSpawnGroup=false
    Offset=&FN_FrontEndVector3Modifier(Gameplay=<-100 0 0>, Front End=<-200 0 0>)
    Offset Mode=::NewEnumerator0
    Override Seed=false
    Rotation Mode=::NewEnumerator1
    Shape Primitive=::NewEnumerator3
    UseOwnerScale=false
   )
   /SetVariables(Particles.Position=&Multiply_Vector(A=Linked() {linkedVariable {name=Particles.Position, type=NiagaraPosition}}, B=<1.2 -1 0.8>)) {isSetParameters=true}
   &AddVelocity(
    Velocity=&FN_FrontEndVector3Modifier(
     Gameplay=&Add_Vector(
      A=<300 0 0>
      B=&Multiply_VectorByFloat(
       Vector=<0 0 0>
       Float=&FloatFromCurve(FloatCurve=DataInterface() {propertyValues {Curve {keys [<at 0 0>,<at 0.6 1>]}}}, CurveIndex=$System.Timeline.Time)
      )
     )
     Front End=<100 0 0>
    )
    Velocity Speed Scale=&FN_FrontEndFloatModifier(Gameplay=1, Front End=1)
   )
  ]
  particleUpdateScript [
   &ParticleState()
   &ScaleSpriteSize(Scale Factor=&Vector2DFromCurve(
     Vector2Curve=DataInterface() {propertyValues {XCurve {keys [<at 0 0.5 RCIM_Cubic>,<at 0.5 1 RCIM_Cubic>]}, YCurve {keys [<at 0 0 RCIM_Cubic>,<at 0.5 1 RCIM_Cubic>]}}}
     ScaleCurve=&Vector2DFromFloat(Value=1)
    ))
   &SolveForcesAndVelocity(Clamp Velocity=false, Limit Acceleration=false)
   &ScaleSpriteSizeBySpeed(Max Scale Factor=<1 50>, Min Scale Factor=<0.01 0.05>, Velocity Threshold=1)
   &ScaleColor(
    Scale Alpha=&FloatFromCurve(
     FloatCurve=DataInterface() {propertyValues {Curve {keys [
         {arriveTangent=6.8397, interpMode=RCIM_Cubic, leaveTangent=6.8397, tangentMode=RCTM_User, time=0, value=0}
         <at 0.3 0.8 RCIM_Cubic>
         {arriveTangent=-2.221, interpMode=RCIM_Cubic, leaveTangent=-2.221, tangentMode=RCTM_User, time=1, value=0}
        ]}}}
     Scale Curve=&FN_FrontEndFloatModifier(
      Gameplay=&Lerp_Float(
       A=0
       B=&FNTimeOfDayScalar(Night Scale=0.2, Morning Scale=0.3, Day Scale=0.4, Evening Scale=0.3)
       Alpha=$System.Timeline003.Float
      )
      Front End=&Lerp_Float(A=0, B=0.3, Alpha=$System.Timeline003.Float)
     )
    )
    Scale RGB=&VectorFromFloat(Value=&FN_FrontEndFloatModifier(Gameplay=1, Front End=1))
    ScaleA=true
    ScaleRGB=true
   )
  ]
  rendererClasses [&NiagaraSpriteRendererProperties]
  renderers [{rendererClass=&NiagaraSpriteRendererProperties}]
 }
 Wingtrails {
  emitterSpawnScript [&FN_RibbonIDOffsetUtility()]
  emitterUpdateScript [
   &EmitterState(
    Inactive Response=::NewEnumerator1
    Life Cycle Mode=::NewEnumerator1
    Loop Behavior=::NewEnumerator1
    Loop Delay=1
    Loop Duration Mode=::NewEnumerator1
    UseLoopDelay=true
   )
   &FN_SpawnAnimTrail(Number of Anim Trails:NiagaraInt32=&MakeCustomIntFromBool(
     Bool=$User.bIsFrontEndPreview
     True INT=2
     False INT=&MakeCustomIntFromBool(Bool=&SetBoolByFloatComparison(A=$System.Timeline.NormalizedTime, B=0.4), True INT=2, False INT=0)
    ))
  ]
  particleSpawnScript [
   &FN_InitializeAnimTrailBoneIndecies(First Bone or Socket Index=1, Second Bone or Socket Index=0)
   &FN_InitializeAnimTrailBoneIndecies(AnimTrail Selection=1, First Bone or Socket Index=3, Second Bone or Socket Index=2) {module=FN_InitializeAnimTrailBoneIndecies001}
   &FN_InitializeAnimTrail(
    AnimTrailWidth=&FN_FrontEndFloatModifier(Gameplay=1, Front End=1.6)
    Color=<rgba 1 1 1 0.25>
    Lifetime=1
    Position Offset=&FN_FrontEndVector3Modifier(
     Gameplay=&MaskVector(Mask Parameter=$Particles.SpawnGroup, Mask=0, Masked Vector=<-4 5 -4>, Passthrough Vector=<-4 -4.5 -4>)
     Front End=&MaskVector(Mask Parameter=$Particles.SpawnGroup, Mask=0, Masked Vector=<0 28 -3>, Passthrough Vector=<0 -25 -3>)
    )
   )
   &AddVelocity(Velocity=&Multiply_VectorByFloat(
     Vector=&MakeVector(
      X=&Lerp_Float(
       A=100
       B=200
       Alpha=&FloatFromCurve(
        FloatCurve=DataInterface() {propertyValues {Curve {keys [{interpMode=RCIM_Cubic, tangentMode=RCTM_User, time=0.5, value=0}, {interpMode=RCIM_Cubic, tangentMode=RCTM_User, time=1, value=1}]}}}
        CurveIndex=$System.Timeline.NormalizedTime
       )
      )
      Y=0
      Z=0
     )
     Float=&FN_FrontEndFloatModifier(Gameplay=1, Front End=1.6)
    ))
  ]
  particleUpdateScript [
   &InterpolateOverTime(Initial Float Value:NiagaraFloat=$System.Timeline.Time, Target Float Value=1, Use Set Interpolation Rate=true)
   &KillParticles(Kill Particles=&SetBoolByFloatComparison(
     A=&FN_FrontEndFloatModifier(Gameplay=$System.Timeline.Time, Front End=1)
     B=$Particles.InterpolateOverTime.InterpolateOverTime_Float.Moving Average
     Comparison Type=::NewEnumerator6
    ))
   &SolveForcesAndVelocity(Clamp Velocity=false, Limit Acceleration=false)
   &FN_UpdateAnimTrail() {module=FN_UpdateAnimTrail001}
   &ScaleColor(
    Scale Alpha=&FloatFromCurve(
     FloatCurve=DataInterface() {propertyValues {Curve {keys [<at 0 1 RCIM_Cubic>,<at 1 0>]}}}
     Scale Curve=&FN_FrontEndFloatModifier(Gameplay=$System.Timeline.Time, Front End=0.6)
    )
    Scale RGB=&MakeVectorFromLinearColorRGB(Color=&FNTimeOfDayColor(
      Night Color=<rgba 0.3086 0.4497 0.75 1>
      Morning Color=<rgba 0.7812 0.5615 0.5615 1>
      Day Color=<rgba 1 1 1 1>
      Evening Color=<rgba 0.7812 0.5615 0.5615 1>
     ))
    ScaleA=true
    ScaleRGB=true
   )
  ]
  rendererClasses [&NiagaraRibbonRendererProperties]
  renderers [{rendererClass=&NiagaraRibbonRendererProperties}]
 }
}
summary {
 systemName=NS_Glider_InflatoDodo_Trail
 emitters [
  {emitterName=Wingtrails, rendererClasses [&NiagaraRibbonRendererProperties]}
  {emitterName=Prop, rendererClasses [&NiagaraMeshRendererProperties]}
  {emitterName=SpeedLines, rendererClasses [&NiagaraSpriteRendererProperties]}
 ]
 userVariables [
  User.bIsFrontEnd {type=NiagaraBool}
  User.bIsFrontEndPreview {type=NiagaraBool}
  User.ForwardDot {type=NiagaraFloat, defaultValue=0, description=""}
  User.RibbonIDOffset {type=NiagaraInt32}
  User.RightDot {type=NiagaraFloat, defaultValue=0, description=""}
 ]
}
